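# Physics asset for the medium male Nerd player body, covering a badge bone (dyn_badge) and the nearby hip area, apparently for secondary badge motion.
# The badge is a single thin box, while the pelvis and left upper thigh twist bone each use two capsules as simplified body volumes.
# Two constraints link the badge and the thigh directly to the pelvis, with matching swing and twist limit angles.
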PhysicsAsset /BRCosmetics/Characters/Player/Male/Medium/Bodies/M_Med_Nerd/Meshes/M_MED_Nerd_Physics
bodies {
 dyn_badge [{
   shapeType=Box
   extentX=0.7667
   extentY=6.1997
   extentZ=10.0234
   shapeName=dyn_badge_box
   center=<5.7318 -0.0654 0.3452>
   rotation=<rot 0 90.8843 -87.1274>
  }]
 pelvis [
  {shapeType=Capsule, length=12, radius=14.25, center=<0 -1.4672 0>, rotation=<rot 90 0 0>}
  {shapeType=Capsule, length=4, radius=6.25, shapeName=pelvis_capsule, center=<0 0 -14.638>, rotation=<rot 90 0 0>}
 ]
 thigh_twist_01_l [
  {shapeType=Capsule, length=19.5, radius=11.75, center=<3.3519 0.0004 0.0001>, rotation=<rot 90 0 0>}
  {shapeType=Capsule, length=4.75, radius=5.5, shapeName=thigh_twist_01_l_capsule, center=<12.3382 2.9743 -10.6524>, rotation=<rot 0 0 89.9999>}
 ]
}
constraints [
 {bone1Name=thigh_twist_01_l, bone2Name=pelvis, swing1LimitDegrees=45, swing2LimitDegrees=45, twistLimitDegrees=45}
 {bone1Name=dyn_badge, bone2Name=pelvis, swing1LimitDegrees=45, swing2LimitDegrees=45, twistLimitDegrees=45}
]
