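# Animation Blueprint for the Nerd male medium body cosmetic, adding secondary motion to badge, tags, scarf, pouch and armband bones.
# It copies the attached parent's pose, then layers AnimDynamics, a pelvis-relative RigidBody and badge adjustments, including a Look At fallback when RigidBody is disabled.
# The EventGraph derives crouch-based badge corrections and drives dyn_spine_05 scarf correction from spine twist and emote state.
# The badge AnimDynamics node is unconnected.
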AnimBlueprint /BRCosmetics/Characters/Player/Male/Medium/Bodies/M_Med_Nerd/Meshes/M_MED_Nerd_body_MESH_Skeleton_AnimBlueprint
target_skeleton=/Game/Characters/Player/Male/Medium/Bodies/M_Med_Nerd/Meshes/M_MED_Nerd_body_MESH_Skeleton
graphs [
 AnimGraph {nodes [
   Anim Dynamics - Bone: dyn_badge(Alpha=0.7, ExternalForce="0,0,0") {
    angularBiasOverride=1
    angularDampingOverride=0.995
    angularSpringConstant=600
    bAngularSpring=true
    bGravityOverrideInSimSpace=true
    bOverrideAngularDamping=true
    bUseSphericalLimits=true
    gravityScale=0.33
    lODThreshold=1
    numSolverIterationsPostUpdate=4
    numSolverIterationsPreUpdate=16
    simulationSpace=RootRelative
    alphaBoolBlend {blendOption=Linear}
    boundBone {boneName=dyn_badge}
    componentAppliedLinearAccClamp=<50 50 50>
    physicsBodyDefinitions [{
      collisionType=CustomSphere
      sphereCollisionRadius=2
      boundBone {boneName=dyn_badge}
      boxExtents=<5 20 20>
      constraintSetup {angularLimitsMax=<25 15 0>, angularLimitsMin=<0 -15 -50>, angularTarget=<1 0 0>}
      localJointOffset=<4 0 0>
     }]
    sphericalLimits [
     {limitRadius=20, drivingBone {boneName=thigh_twist_01_l}, sphereLocalOffset=<20 -8.253 8>}
     {limitRadius=10, drivingBone {boneName=thigh_l}, sphereLocalOffset=<-1 0.5 2>}
    ]
   }
   n1 = "Constraint - dyn_spine_05 (PPPP)"!(
    ComponentPose=Local To Component(LocalPose=Copy Pose From Mesh() {bUseAttachedParent=true})
    Alpha=Get scarf_frontend_emote_alpha() -> (scarf_frontend_emote_alpha:double)
    ConstraintWeights_0=0.092
    ConstraintWeights_1:float=0.267
    ConstraintWeights_2:float=0.578
    ConstraintWeights_3:float=0.063
   ) -> (Pose) {
    alphaBoolBlend {blendOption=Linear}
    boneToModify {boneName=dyn_spine_05}
    constraintSetup [{targetBone {boneName=spine_02}}, {targetBone {boneName=spine_03}}, {targetBone {boneName=spine_04}}, {targetBone {boneName=spine_05}}]
    constraintWeights [1,1,1,1]
   }
   n2 = Anim Dynamics - Bone: dyn_tags(ComponentPose=@n1, Alpha=0.7, ExternalForce="0,0,0") {
    angularBiasOverride=1
    angularDampingOverride=0.98
    bGravityOverrideInSimSpace=true
    lODThreshold=1
    numSolverIterationsPostUpdate=4
    numSolverIterationsPreUpdate=16
    alphaBoolBlend {blendOption=Linear}
    boundBone {boneName=dyn_tags}
    physicsBodyDefinitions [{
      sphereCollisionRadius=0
      boundBone {boneName=dyn_tags}
      boxExtents=<5 2 10>
      constraintSetup {angularTargetAxis=AxisZ, twistAxis=AxisZ, angularLimitsMax=<25 45 0>, angularLimitsMin=<0 -45 0>, angularTarget=<0 0 0>}
      localJointOffset=<0 1 -1>
     }]
   }
   n3 = RigidBody!(ComponentPose=@n2, bAlphaBoolEnabled:bool=Get bRigidBodyEnabled(), ExternalForce) -> (Pose) {
    alphaInputType=Bool
    overridePhysicsAsset=&/M_MED_Nerd_Physics
    simulationSpace=BaseBoneSpace
    alphaBoolBlend {blendInTime=0.5, blendOption=Linear}
    baseBoneRef {boneName=pelvis}
    componentAppliedLinearAccClamp=<3000 3000 3000>
    componentLinearAccScale=<0.1 0.1 0.1>
    componentLinearVelScale=<0.1 0.1 0.1>
   }
   n4 = Look At - Bone: dyn_badge!(
    ComponentPose=@n3
    bAlphaBoolEnabled:bool=NOT Boolean(A=Get bRigidBodyEnabled())
    LookAtLocation:Vector=Get badge_translation_crouch_lookat() -> (badge_translation_crouch_lookat:Vector)
   ) -> (Pose) {
    alphaInputType=Bool
    alphaBoolBlend {blendOption=Linear, blendOutTime=0.5}
    boneToModify {boneName=dyn_badge}
    lookAtLocation=<3.75 10 -9>
    lookAtTarget {boneReference {boneName=calf_l}}
    lookAt_Axis {axis=<1 0>}
   }
   n5 = "Transform (Modify) Bone - Bone: dyn_badge"!(
    ComponentPose=@n4
    bAlphaBoolEnabled:bool=Get badge_right_crouch_movement_alpha() -> (badge_right_crouch_movement_alpha:bool)
    Translation="0,-2.5,-0.000"
    Rotation="0,0,0"
    Scale="1,1,1"
   ) -> (Pose) {
    alphaInputType=Bool
    translationMode=BMM_Additive
    translationSpace=BCS_BoneSpace
    alphaBoolBlend {blendInTime=0.7, blendOption=Linear, blendOutTime=0.7}
    boneToModify {boneName=dyn_badge}
    translation {y=-2.5}
   }
   n6 = "Transform (Modify) Bone - Bone: dyn_badge"(ComponentPose=@n5, bAlphaBoolEnabled=True, Translation="-5,0,0", Rotation="0,0,0", Scale="0.5,0.5,0.5") {
    alphaInputType=Bool
    rotationSpace=BCS_BoneSpace
    scaleMode=BMM_Additive
    translationMode=BMM_Additive
    translationSpace=BCS_BoneSpace
    alphaBoolBlend {blendInTime=0.2, blendOption=Linear, blendOutTime=0.2}
    boneToModify {boneName=dyn_badge}
    scale=<0.5 0.5 0.5>
    translation {x=-5}
   }
   n7 = Get bIsInFrontEnd()
   n8 = float * float(A=Get PawnSpeed2D(), B="To Float (Boolean)"(InBool=NOT Boolean(A=@n7))) {is=title K2Node_PromotableOperator}
   n9 = Anim Dynamics - Bone: dyn_scarf_r(ComponentPose=@n6, Alpha=@n8, ExternalForce="0,0,0") {
    bGravityOverrideInSimSpace=true
    gravityScale=0.1
    lODThreshold=1
    alphaBoolBlend {blendOption=Linear}
    alphaScaleBiasClamp {bClampResult=true, clampMax=0.09, clampMin=0.05, scale=0.0033}
    boundBone {boneName=dyn_scarf_r}
    physicsBodyDefinitions [{
      sphereCollisionRadius=0
      boundBone {boneName=dyn_scarf_r}
      boxExtents=<2 2 2>
      constraintSetup {angularLimitsMax=<0 0 0>, angularLimitsMin=<0 0 0>, angularTarget=<0 0 0>, linearAxesMax=<2 2 2>, linearAxesMin=<-2 -2 -2>}
      localJointOffset=<0 0 -1>
     }]
   }
   n10 = Anim Dynamics - Bone: dyn_scarf_l(ComponentPose=@n9, Alpha=@n8, ExternalForce="0,0,0") {
    bGravityOverrideInSimSpace=true
    gravityScale=0.1
    lODThreshold=1
    alphaBoolBlend {blendOption=Linear}
    alphaScaleBiasClamp {bClampResult=true, clampMax=0.09, clampMin=0.05, scale=0.0033}
    boundBone {boneName=dyn_scarf_l}
    physicsBodyDefinitions [{
      sphereCollisionRadius=0
      boundBone {boneName=dyn_scarf_l}
      boxExtents=<2 2 2>
      constraintSetup {angularLimitsMax=<0 0 0>, angularLimitsMin=<0 0 0>, angularTarget=<0 0 0>, linearAxesMax=<2 2 2>, linearAxesMin=<-2 -2 -2>}
      localJointOffset=<0 0 -1>
     }]
   }
   n11 = Anim Dynamics - Bone: dyn_scarf_r(ComponentPose=@n10, Alpha=@n8, ExternalForce="0,0,0") {
    bGravityOverrideInSimSpace=true
    gravityScale=0.1
    lODThreshold=1
    alphaBoolBlend {blendOption=Linear}
    alphaScaleBiasClamp {bClampResult=true, clampMax=0.09, clampMin=0.05, scale=0.0033}
    boundBone {boneName=dyn_scarf_r}
    physicsBodyDefinitions [{
      sphereCollisionRadius=0
      boundBone {boneName=dyn_scarf_r}
      boxExtents=<2 2 2>
      constraintSetup {angularLimitsMax=<0 0 0>, angularLimitsMin=<0 0 0>, angularTarget=<0 0 0>, linearAxesMax=<2 2 2>, linearAxesMin=<-2 -2 -2>}
      localJointOffset=<0 0 -1>
     }]
   }
   n12 = Anim Dynamics - Bone: dyn_scarf_b(ComponentPose=@n11, Alpha=@n8, ExternalForce="0,0,0") {
    bGravityOverrideInSimSpace=true
    gravityScale=0.1
    lODThreshold=1
    alphaBoolBlend {blendOption=Linear}
    alphaScaleBiasClamp {bClampResult=true, clampMax=0.09, clampMin=0.05, scale=0.0033}
    boundBone {boneName=dyn_scarf_b}
    physicsBodyDefinitions [{
      sphereCollisionRadius=0
      boundBone {boneName=dyn_scarf_b}
      boxExtents=<2 2 2>
      constraintSetup {angularLimitsMax=<0 0 0>, angularLimitsMin=<0 0 0>, angularTarget=<0 0 0>, linearAxesMax=<2 2 2>, linearAxesMin=<-2 -2 -2>}
      localJointOffset=<0 0 -1>
     }]
   }
   n13 = Anim Dynamics - Bone: dyn_pouch_r(ComponentPose=@n12, Alpha=1, ExternalForce="0,0,0") {
    bGravityOverrideInSimSpace=true
    gravityScale=0.1
    lODThreshold=1
    alphaBoolBlend {blendOption=Linear}
    alphaScaleBiasClamp {bClampResult=true, clampMax=0.33, clampMin=0.1, scale=0.0033}
    boundBone {boneName=dyn_pouch_r}
    physicsBodyDefinitions [{
      sphereCollisionRadius=0
      boundBone {boneName=dyn_pouch_r}
      boxExtents=<5 5 5>
      constraintSetup {angularTargetAxis=AxisY, twistAxis=AxisY, angularLimitsMax=<0 0 0>, angularLimitsMin=<0 0 0>, angularTarget=<0 0 0>, linearAxesMax=<0 0 2>}
      localJointOffset=<0 0 5>
     }]
   }
   n14 = Anim Dynamics - Bone: dyn_armband_knot(ComponentPose=@n13, Alpha=0.75, ExternalForce="0,0,0") {
    alpha=0.9
    angularDampingOverride=0.98
    angularSpringConstant=1500
    bAngularSpring=true
    bGravityOverrideInSimSpace=true
    bOverrideAngularDamping=true
    numSolverIterationsPostUpdate=4
    numSolverIterationsPreUpdate=16
    simulationSpace=Actor
    alphaBoolBlend {blendOption=Linear}
    boundBone {boneName=dyn_armband_knot}
    physicsBodyDefinitions [{
      sphereCollisionRadius=0
      boundBone {boneName=dyn_armband_knot}
      boxExtents=<15 5 5>
      constraintSetup {angularTargetAxis=AxisZ, twistAxis=AxisY, angularLimitsMax=<15 0 15>, angularLimitsMin=<-15 0 -15>, angularTarget=<0 0 1>}
      localJointOffset=<0 0 0>
     }]
   }
   n15 = Anim Dynamics - Bone: dyn_armband_low(ComponentPose=@n14, Alpha=0.5, ExternalForce="0,0,0") {
    alpha=0.9
    angularDampingOverride=0.98
    angularSpringConstant=1500
    bAngularSpring=true
    bGravityOverrideInSimSpace=true
    bOverrideAngularDamping=true
    numSolverIterationsPostUpdate=4
    numSolverIterationsPreUpdate=16
    simulationSpace=Actor
    alphaBoolBlend {blendOption=Linear}
    boundBone {boneName=dyn_armband_low}
    physicsBodyDefinitions [{
      sphereCollisionRadius=0
      boundBone {boneName=dyn_armband_low}
      boxExtents=<15 5 5>
      constraintSetup {angularTargetAxis=AxisZ, twistAxis=AxisY, angularLimitsMax=<15 15 15>, angularLimitsMin=<-15 -15 -15>, angularTarget=<0 0 1>}
      localJointOffset=<0 0 0>
     }]
   }
   n16 = Anim Dynamics - Bone: dyn_armband_hi(ComponentPose=@n15, Alpha=0.5, ExternalForce="0,0,0") {
    alpha=0.9
    angularDampingOverride=0.98
    angularSpringConstant=1500
    bAngularSpring=true
    bGravityOverrideInSimSpace=true
    bOverrideAngularDamping=true
    numSolverIterationsPostUpdate=4
    numSolverIterationsPreUpdate=16
    simulationSpace=Actor
    alphaBoolBlend {blendOption=Linear}
    boundBone {boneName=dyn_armband_hi}
    physicsBodyDefinitions [{
      sphereCollisionRadius=0
      boundBone {boneName=dyn_armband_hi}
      boxExtents=<15 5 5>
      constraintSetup {angularTargetAxis=AxisZ, twistAxis=AxisY, angularLimitsMax=<15 15 15>, angularLimitsMin=<-15 -15 -15>, angularTarget=<0 0 1>}
      localJointOffset=<0 0 0>
     }]
   }
   n17 = "Transform (Modify) Bone - Bone: dyn_spine_05"(ComponentPose=@n16, Alpha=Get scarf_alpha() -> (scarf_alpha:double), Translation="0,4,0", Rotation="0,0,-12.5", Scale="1,1,1") {
    rotationMode=BMM_Additive
    translationMode=BMM_Additive
    alphaBoolBlend {blendOption=Linear}
    boneToModify {boneName=dyn_spine_05}
    rotation {roll=-12.5}
    translation {y=4}
   }
   Output Pose(Result=Component To Local(ComponentPose=@n17)) {layerGroup=None}
  ]}
 EventGraph {nodes [
   n1 = Event BlueprintUpdateAnimation()
   > n2 = Sequence()
   n3 = GetOwnerSkeletalMeshComponent()
   n4 = BreakTransform(InTransform=GetDeltaTransformFromRefPose(self=@n3, BoneName=spine_05, BaseName=spine_01))
   > Set scarf_alpha(scarf_alpha:double=MapRangeClamped(Value=BreakRotator(InRot=@n4.Rotation).Yaw, InRangeA=0, InRangeB=15)) -> (Output_Get:double)
   > Set badge_translation_crouch_lookat(badge_translation_crouch_lookat:Vector=SelectVector(A="40,40,-9", B="9,10,-9", bPickA=Get bIsCrouching())) -> (Output_Get:Vector)
   n5 = Vector Length(A=Get PawnVelocity())
   > Set badge_idle_crouch_alpha(badge_idle_crouch_alpha:bool=AND Boolean!(A="float < float"(A=@n5, B=5), B=Get bIsCrouching(), C:bool=Get bRigidBodyEnabled()) -> (ReturnValue)) -> (Output_Get:bool)
   n6 = Get PawnVelocity()
   n7 = TryGetPawnOwner()
   n8 = "Clamp (Float)"(
    Value=Dot Product!(
     A:Vector=Normalize!(A:Vector=@n6, Tolerance:float=0.0001) -> (ReturnValue:Vector) {is=title K2Node_CallFunction}
     B:Vector=GetActorRightVector!(self:Actor Object Reference=@n7) -> (ReturnValue:Vector)
    ) -> (ReturnValue:double)
    Min=0
    Max=1
   )
   > Set badge_right_crouch_movement_alpha(badge_right_crouch_movement_alpha:bool=AND Boolean!(A=Get bIsCrouching(), B="float > float"(A=@n8)) -> (ReturnValue)) -> (Output_Get:bool)
   Set DeltaTimeX(execute=@n2.then_1, DeltaTimeX:double=@n1.DeltaTimeX) -> (Output_Get:double)
   > n9 = Branch(Condition=Get bIsPlayingEmote())
   > Set scarf_frontend_emote_alpha(scarf_frontend_emote_alpha:double=FInterpTo_Constant!(
     Current:double=Get scarf_frontend_emote_alpha()
     Target:double=1
     DeltaTime:double=Get DeltaTimeX() -> (DeltaTimeX:double)
     InterpSpeed:double=6
    ) -> (ReturnValue:double)) -> (Output_Get:double)
   Set scarf_frontend_emote_alpha(
    execute=@n9.else
    scarf_frontend_emote_alpha=FInterpTo_Constant(Current=Get scarf_frontend_emote_alpha(), Target=0, DeltaTime=Get DeltaTimeX(), InterpSpeed=16)
   )
   Event BlueprintInitializeAnimation()
   > Set MeshToCopy(MeshToCopy:Skeletal Mesh Component Object Reference=@n3) -> (Output_Get:Skeletal Mesh Component Object Reference)
  ]}
]
variables [
 badge_idle_crouch_alpha {type=boolean}
 badge_right_crouch_movement_alpha {type=boolean}
 badge_translation_crouch_lookat {type=Vector}
 DeltaTimeX {type=number}
 MeshToCopy {type=SkeletalMeshComponent}
 scarf_alpha {type=number}
 scarf_frontend_emote_alpha {type=number}
]
